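# Material function for glTF Exporter proxy materials that outputs BaseColor and Opacity from a texture multiplied by a color factor.
# MF_GLTF_UVTransform builds the UVs from parameters for UV index, offset, scale, rotation and tiling method, with both offset and scale axes wired to the R channel.
# The texture uses derivative mip selection and shared clamp sampling, and defaults to white.
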MaterialFunction /GLTFExporter/Functions/Proxy/MF_GLTF_BaseColor
graph {nodes [
  n1 = VectorParameter() {group=Base Color, parameterName=Base Color UV Offset, sortPriority=3}
  n2 = VectorParameter() {group=Base Color, parameterName=Base Color UV Scale, sortPriority=4, defaultValue=<rgba 1 1 0 0>}
  n3 = MaterialFunctionCall!(
   TexCoord=ScalarParameter() {group=Base Color, parameterName=Base Color UV Index, sliderMax=7, sortPriority=2}
   Offset_X=@n1.R
   Offset_Y=@n1.R
   Scale_X=@n2.R
   Scale_Y=@n2.R
   Rotation=ScalarParameter() {group=Base Color, parameterName=Base Color UV Rotation, sliderMax=6.2832, sortPriority=5}
   TilingMethod=VectorParameter() {parameterName=Base Color Tiling Method}.RGB
  ) {materialFunction=&/MF_GLTF_UVTransform}
  n4 = TextureSampleParameter2D!(UVs=@n3.UVs, "DDX(UVs)"=@n3.UVs, "DDY(UVs)"=@n3.UVs, Apply View MipBias) {
   group=Base Color
   mipValueMode=TMVM_Derivative
   parameterName=Base Color Texture
   samplerSource=SSM_Clamp_WorldGroupSettings
   sortPriority=1
   texture=/GLTFExporter/Textures/Proxy/T_GLTF_DefaultWhite
   channelNames {a=Opacity, b="", g="", r=""}
  }
  n5 = VectorParameter() {group=Base Color, parameterName=Base Color Factor, defaultValue=<rgba 1 1 1 1>}
  FunctionOutput(None=Multiply(A=@n4.RGB, B=@n5.RGB)) {outputName=BaseColor}
  FunctionOutput(None=Multiply(A=@n4.Opacity, B=@n5.A)) {outputName=Opacity, sortPriority=1}
 ]}
