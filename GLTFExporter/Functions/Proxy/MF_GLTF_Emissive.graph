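# Material function for glTF Exporter proxy materials that outputs emissive color as texture RGB times Emissive Factor times Emissive Strength.
# MF_GLTF_UVTransform builds the UVs from parameters for UV index, offset, scale, rotation and tiling method, with offset and scale X and Y both wired to the R channel.
# The clamp-sampled texture defaults to white and uses derivative mip mode, with those UVs also feeding DDX and DDY.
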MaterialFunction /GLTFExporter/Functions/Proxy/MF_GLTF_Emissive
graph {nodes [
  n1 = VectorParameter() {group=Emissive, parameterName=Emissive UV Offset, sortPriority=3}
  n2 = VectorParameter() {group=Emissive, parameterName=Emissive UV Scale, sortPriority=4, defaultValue=<rgba 1 1 0 0>}
  n3 = MaterialFunctionCall!(
   TexCoord=ScalarParameter() {group=Emissive, parameterName=Emissive UV Index, sliderMax=7, sortPriority=2}
   Offset_X=@n1.R
   Offset_Y=@n1.R
   Scale_X=@n2.R
   Scale_Y=@n2.R
   Rotation=ScalarParameter() {group=Emissive, parameterName=Emissive UV Rotation, sliderMax=6.2832, sortPriority=5}
   TilingMethod=VectorParameter() {group=Emissive, parameterName=Emissive Tiling Method, sortPriority=3}.RGB
  ) {materialFunction=&/MF_GLTF_UVTransform}
  n4 = TextureSampleParameter2D!(UVs=@n3.UVs, "DDX(UVs)"=@n3.UVs, "DDY(UVs)"=@n3.UVs, Apply View MipBias) {
   group=Emissive
   mipValueMode=TMVM_Derivative
   parameterName=Emissive Texture
   samplerSource=SSM_Clamp_WorldGroupSettings
   sortPriority=1
   texture=/GLTFExporter/Textures/Proxy/T_GLTF_DefaultWhite
  }
  n5 = VectorParameter() {group=Emissive, parameterName=Emissive Factor, defaultValue=<rgba 0 0 0 1>}
  n6 = ScalarParameter() {defaultValue=1, group=Emissive, parameterName=Emissive Strength}
  FunctionOutput(None=Multiply(A=@n4.RGB, B=Multiply(A=@n5.RGB, B=@n6))) {outputName=Emissive}
 ]}
