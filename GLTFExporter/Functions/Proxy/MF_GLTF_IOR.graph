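# MF_GLTF_IOR is a GLTFExporter proxy material function for the glTF index of refraction.
# Its output is a scalar parameter named IOR in the Refraction group, defaulting to 1.5.
# It takes no inputs and does no other math, so the parameter value passes straight to the function output.
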MaterialFunction /GLTFExporter/Functions/Proxy/MF_GLTF_IOR
graph {nodes [FunctionOutput(None=ScalarParameter() {defaultValue=1.5, group=Refraction, parameterName=IOR})]}
