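# Material function for glTF exporter proxy materials that adds thin-film iridescence to an incoming BaseColor.
# A thickness texture's green channel interpolates minimum to maximum thickness, which is scaled down and fed with both IORs into Interchange's MF_Iridescence.
# The result is half-blended with BaseColor, then weighted by Iridescence Factor times the iridescence texture's red channel.
# Both textures get parameterized UV index, offset, scale, rotation and tiling through MF_GLTF_UVTransform.
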MaterialFunction /GLTFExporter/Functions/Proxy/MF_GLTF_Iridescence
graph {nodes [
  n1 = FunctionInput() {inputName=BaseColor}
  n2 = ScalarParameter() {defaultValue=100, group=Iridescence, parameterName=IridescenceThickness Minimum, sortPriority=1}
  n3 = ScalarParameter() {defaultValue=400, group=Iridescence, parameterName=IridescenceThickness Maximum, sortPriority=2}
  n4 = VectorParameter() {group=Iridescence, parameterName=IridescenceThickness UV Offset, sortPriority=6}
  n5 = VectorParameter() {group=Iridescence, parameterName=IridescenceThickness UV Scale, sortPriority=7, defaultValue=<rgba 1 1 0 0>}
  n6 = MaterialFunctionCall!(
   TexCoord=ScalarParameter() {group=Iridescence, parameterName=IridescenceThickness UV Index, sliderMax=7, sortPriority=5}
   Offset_X=@n4.R
   Offset_Y=@n4.R
   Scale_X=@n5.R
   Scale_Y=@n5.R
   Rotation=ScalarParameter() {group=Iridescence, parameterName=IridescenceThickness UV Rotation, sliderMax=6.2832, sortPriority=8}
   TilingMethod=VectorParameter() {group=Iridescence, parameterName=IridescenceThickness Tiling Method, sortPriority=6}.RGB
  ) {materialFunction=&/MF_GLTF_UVTransform}
  n7 = TextureSampleParameter2D!(UVs=@n6.UVs, "DDX(UVs)"=@n6.UVs, "DDY(UVs)"=@n6.UVs, Apply View MipBias) {
   group=Iridescence
   mipValueMode=TMVM_Derivative
   parameterName=IridescenceThickness Texture
   samplerSource=SSM_Clamp_WorldGroupSettings
   sortPriority=4
   texture=/GLTFExporter/Textures/Proxy/T_GLTF_DefaultWhite
  }
  n8 = FunctionInput() {description=BaseMaterial's IOR, inputName=IOR, inputType=FunctionInput_Scalar, sortPriority=1, previewValue=<1.5 0 0 1>}
  n9 = MaterialFunctionCall!(
   Thickness=Divide(A=LinearInterpolate(A=@n2, B=@n3, Alpha=@n7.G)) {constB=10000}
   n2=ScalarParameter() {defaultValue=1.3, group=Iridescence, parameterName=Iridescence IOR, sortPriority=0}
   n3=@n8
   k=Constant() {r=1}
  ) {materialFunction=/InterchangeAssets/Functions/MF_Iridescence}
  n10 = ScalarParameter() {group=Iridescence, parameterName=Iridescence Factor, sortPriority=3}
  n11 = VectorParameter() {group=Iridescence, parameterName=Iridescence UV Offset, sortPriority=11}
  n12 = VectorParameter() {group=Iridescence, parameterName=Iridescence UV Scale, sortPriority=12, defaultValue=<rgba 1 1 0 0>}
  n13 = MaterialFunctionCall(
   TexCoord=ScalarParameter() {group=Iridescence, parameterName=Iridescence UV Index, sliderMax=7, sortPriority=10}
   Offset_X=@n11.R
   Offset_Y=@n11.R
   Scale_X=@n12.R
   Scale_Y=@n12.R
   Rotation=ScalarParameter() {group=Iridescence, parameterName=Iridescence UV Rotation, sliderMax=6.2832, sortPriority=13}
   TilingMethod=VectorParameter() {group=Iridescence, parameterName=Iridescence Tiling Method, sortPriority=11}.RGB
  ) {materialFunction=&MF_GLTF_UVTransform}
  n14 = TextureSampleParameter2D(UVs=@n13.UVs, "DDX(UVs)"=@n13.UVs, "DDY(UVs)"=@n13.UVs) {
   group=Iridescence
   mipValueMode=TMVM_Derivative
   parameterName=Iridescence Texture
   samplerSource=SSM_Clamp_WorldGroupSettings
   sortPriority=9
   texture=&T_GLTF_DefaultWhite
  }
  FunctionOutput(None=LinearInterpolate(A=@n1, B=LinearInterpolate(A=@n1, B=@n9.Result), Alpha=Multiply(A=@n10, B=@n14.R))) {outputName=BaseColor}
 ]}
