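# Material function for glTF exporter proxy materials that outputs Metallic and Roughness from one packed linear texture.
# MF_GLTF_UVTransform builds UVs from index, offset, scale, rotation and tiling parameters, with the R channel driving both offset and scale axes.
# Those UVs feed the clamped, derivative-mip sampler's coordinate and both derivative inputs.
# Blue (metallic) and green (roughness) are multiplied by their factor parameters, over a white default texture.
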MaterialFunction /GLTFExporter/Functions/Proxy/MF_GLTF_MetallicRoughness
graph {nodes [
  n1 = VectorParameter() {group=Metallic Roughness, parameterName=Metallic Roughness UV Offset, sortPriority=4}
  n2 = VectorParameter() {group=Metallic Roughness, parameterName=Metallic Roughness UV Scale, sortPriority=5, defaultValue=<rgba 1 1 0 0>}
  n3 = MaterialFunctionCall!(
   TexCoord=ScalarParameter() {group=Metallic Roughness, parameterName=Metallic Roughness UV Index, sliderMax=7, sortPriority=3}
   Offset_X=@n1.R
   Offset_Y=@n1.R
   Scale_X=@n2.R
   Scale_Y=@n2.R
   Rotation=ScalarParameter() {group=Metallic Roughness, parameterName=Metallic Roughness UV Rotation, sliderMax=6.2832, sortPriority=6}
   TilingMethod=VectorParameter() {group=Metallic Roughness, parameterName=Metallic Roughness Tiling Method, sortPriority=4}.RGB
  ) {materialFunction=&/MF_GLTF_UVTransform}
  n4 = TextureSampleParameter2D!(UVs=@n3.UVs, "DDX(UVs)"=@n3.UVs, "DDY(UVs)"=@n3.UVs, Apply View MipBias) {
   group=Metallic Roughness
   mipValueMode=TMVM_Derivative
   parameterName=Metallic Roughness Texture
   samplerSource=SSM_Clamp_WorldGroupSettings
   samplerType=SAMPLERTYPE_LinearColor
   sortPriority=2
   texture=/GLTFExporter/Textures/Proxy/T_GLTF_DefaultWhiteLinear
   channelNames {a="", b=Metallic, g=Roughness, r=""}
  }
  FunctionOutput(None=Multiply(A=ScalarParameter() {defaultValue=1, group=Metallic Roughness, parameterName=Roughness Factor, sliderMax=1, sortPriority=1}, B=@n4.Roughness)) {outputName=Roughness, sortPriority=1}
  FunctionOutput(None=Multiply(A=@n4.Metallic, B=ScalarParameter() {defaultValue=1, group=Metallic Roughness, parameterName=Metallic Factor, sliderMax=1})) {outputName=Metallic}
 ]}
