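# Material function for glTF proxy materials in GLTFExporter that outputs a normal from a parameterized normal map.
# MF_GLTF_UVTransform builds UVs from index, offset, scale, rotation and tiling parameters, with X and Y both reading red channels of offset and scale.
# Those UVs drive the texture sample and its derivative mip inputs.
# Normal Scale multiplies the sampled X and Y, leaving Z unscaled, before normalization.
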MaterialFunction /GLTFExporter/Functions/Proxy/MF_GLTF_Normal
graph {nodes [
  n1 = VectorParameter() {group=Normal, parameterName=Normal UV Offset, sortPriority=3}
  n2 = VectorParameter() {group=Normal, parameterName=Normal UV Scale, sortPriority=4, defaultValue=<rgba 1 1 0 0>}
  n3 = MaterialFunctionCall!(
   TexCoord=ScalarParameter() {group=Normal, parameterName=Normal UV Index, sliderMax=7, sortPriority=2}
   Offset_X=@n1.R
   Offset_Y=@n1.R
   Scale_X=@n2.R
   Scale_Y=@n2.R
   Rotation=ScalarParameter() {group=Normal, parameterName=Normal UV Rotation, sliderMax=6.2832, sortPriority=5}
   TilingMethod=VectorParameter() {group=Normal, parameterName=Normal Tiling Method, sortPriority=3}.RGB
  ) {materialFunction=&/MF_GLTF_UVTransform}
  n4 = TextureSampleParameter2D!(UVs=@n3.UVs, "DDX(UVs)"=@n3.UVs, "DDY(UVs)"=@n3.UVs, Apply View MipBias) {
   group=Normal
   mipValueMode=TMVM_Derivative
   parameterName=Normal Texture
   samplerSource=SSM_Clamp_WorldGroupSettings
   samplerType=SAMPLERTYPE_Normal
   sortPriority=1
   texture=/GLTFExporter/Textures/Proxy/T_GLTF_DefaultNormal
  }
  n5 = ScalarParameter() {defaultValue=1, group=Normal, parameterName=Normal Scale}
  n6 = MaterialFunctionCall(X=@n5, Y=@n5, Z=Constant() {r=1}) {materialFunction=&MakeFloat3}
  FunctionOutput(None=Normalize(VectorInput=Multiply(A=@n4.RGB, B=@n6.Result))) {outputName=Normal}
 ]}
