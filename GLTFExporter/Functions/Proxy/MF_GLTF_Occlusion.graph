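# GLTFExporter proxy material function that outputs ambient occlusion from a texture parameter with a white linear default.
# MF_GLTF_UVTransform builds the sampling UVs from index, offset, scale, rotation and tiling parameters, with each red channel of offset and scale driving both X and Y.
# The texture samples with derivative mip mode and the shared clamp sampler.
# Its red channel is multiplied by a strength scalar, rather than lerped toward white.
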MaterialFunction /GLTFExporter/Functions/Proxy/MF_GLTF_Occlusion
graph {nodes [
  n1 = VectorParameter() {group=Occlusion, parameterName=Occlusion UV Offset, sortPriority=3}
  n2 = VectorParameter() {group=Occlusion, parameterName=Occlusion UV Scale, sortPriority=4, defaultValue=<rgba 1 1 0 0>}
  n3 = MaterialFunctionCall!(
   TexCoord=ScalarParameter() {group=Occlusion, parameterName=Occlusion UV Index, sliderMax=7, sortPriority=2}
   Offset_X=@n1.R
   Offset_Y=@n1.R
   Scale_X=@n2.R
   Scale_Y=@n2.R
   Rotation=ScalarParameter() {group=Occlusion, parameterName=Occlusion UV Rotation, sliderMax=6.2832, sortPriority=5}
   TilingMethod=VectorParameter() {group=Occlusion, parameterName=Occlusion Tiling Method, sortPriority=3}.RGB
  ) {materialFunction=&/MF_GLTF_UVTransform}
  n4 = TextureSampleParameter2D!(UVs=@n3.UVs, "DDX(UVs)"=@n3.UVs, "DDY(UVs)"=@n3.UVs, Apply View MipBias) {
   group=Occlusion
   mipValueMode=TMVM_Derivative
   parameterName=Occlusion Texture
   samplerSource=SSM_Clamp_WorldGroupSettings
   samplerType=SAMPLERTYPE_LinearColor
   sortPriority=1
   texture=/GLTFExporter/Textures/Proxy/T_GLTF_DefaultWhiteLinear
   channelNames {a="", b="", g="", r=Occlusion}
  }
  FunctionOutput(None=Multiply(A=@n4.Occlusion, B=ScalarParameter() {defaultValue=1, group=Occlusion, parameterName=Occlusion Strengh, sliderMax=1})) {outputName=Occlusion}
 ]}
