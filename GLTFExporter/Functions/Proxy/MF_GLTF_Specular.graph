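# Material function in the GLTFExporter proxy set that outputs a Specular value from exposed Specular parameters.
# MF_GLTF_UVTransform builds UVs from index, offset, scale, rotation and tiling parameters, with red channels alone driving both offset and scale axes.
# Its UVs output feeds the coordinate and both derivative inputs of a derivative-mip texture sample, defaulting to linear white.
# The sampled alpha is multiplied by Specular Factor.
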MaterialFunction /GLTFExporter/Functions/Proxy/MF_GLTF_Specular
graph {nodes [
  n1 = VectorParameter() {group=Specular, parameterName=Specular UV Offset, sortPriority=3}
  n2 = VectorParameter() {group=Specular, parameterName=Specular UV Scale, sortPriority=4, defaultValue=<rgba 1 1 0 0>}
  n3 = MaterialFunctionCall!(
   TexCoord=ScalarParameter() {group=Specular, parameterName=Specular UV Index, sliderMax=7, sortPriority=2}
   Offset_X=@n1.R
   Offset_Y=@n1.R
   Scale_X=@n2.R
   Scale_Y=@n2.R
   Rotation=ScalarParameter() {group=Specular, parameterName=Specular UV Rotation, sliderMax=6.2832, sortPriority=5}
   TilingMethod=VectorParameter() {group=Specular, parameterName=Specular Tiling Method, sortPriority=3}.RGB
  ) {materialFunction=&/MF_GLTF_UVTransform}
  n4 = TextureSampleParameter2D!(UVs=@n3.UVs, "DDX(UVs)"=@n3.UVs, "DDY(UVs)"=@n3.UVs, Apply View MipBias) {
   group=Specular
   mipValueMode=TMVM_Derivative
   parameterName=Specular Texture
   samplerSource=SSM_Clamp_WorldGroupSettings
   samplerType=SAMPLERTYPE_LinearColor
   sortPriority=1
   texture=/GLTFExporter/Textures/Proxy/T_GLTF_DefaultWhiteLinear
  }
  FunctionOutput(None=Multiply(A=ScalarParameter() {defaultValue=0.5, group=Specular, parameterName=Specular Factor, sortPriority=0}, B=@n4.A)) {outputName=Specular}
 ]}
