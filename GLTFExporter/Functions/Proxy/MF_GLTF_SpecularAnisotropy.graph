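# Material function for glTF proxy materials in GLTFExporter that outputs anisotropy Tangent and strength from parameters.
# A direction texture, sampled through MF_GLTF_UVTransform, supplies red-green direction and blue strength, multiplied by Anisotropy Strength.
# Custom HLSL rotates the remapped direction and applies a tangent/bitangent/normal basis, using the optional Normal input.
# Tangent and default normal use constant axes under MATERIAL_TANGENTSPACENORMAL, otherwise normalized world-space vertex vectors.
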MaterialFunction /GLTFExporter/Functions/Proxy/MF_GLTF_SpecularAnisotropy
graph {nodes [
  MaterialExpressionScalarParameter_19 = ScalarParameter() {group=Anisotropy, parameterName=Anisotropy Rotation, sortPriority=1}
  MaterialExpressionVertexTangentWS_1 = VertexTangentWS()
  MaterialExpressionCustom_1 = Custom!(SurfaceTangent=@MaterialExpressionVertexTangentWS_1) {
   code="{\r\n##if MATERIAL_TANGENTSPACENORMAL\r\n  const float3 T = float3(1,0,0);\r\n##else\r\n  const float3 T = normalize(SurfaceTangent);\r\n##endif\r\n  return T;\r\n}"
   description=GetTangent
   showCode=true
   inputs [{inputName=SurfaceTangent, input {expression=&self:MaterialExpressionVertexTangentWS_1}}]
  }
  MaterialExpressionVertexNormalWS_0 = VertexNormalWS()
  n1 = Custom(SurfaceTangent=@MaterialExpressionVertexNormalWS_0) {
   code="{\r\n##if MATERIAL_TANGENTSPACENORMAL\r\n  const float3 T = float3(0,0,1);\r\n##else\r\n  const float3 T = normalize(SurfaceTangent);\r\n##endif\r\n  return T;\r\n}"
   description=GetNormal
   showCode=true
   inputs [{inputName=SurfaceTangent, input {expression=&self:MaterialExpressionVertexNormalWS_0}}]
  }
  MaterialExpressionFunctionInput_2 = FunctionInput(Preview=@n1) {bUsePreviewValueAsDefault=true, inputName=Normal, sortPriority=3, previewValue=<0 0 1 1>}
  n2 = VectorParameter() {group=Anisotropy, parameterName=Anisotropy UV Offset, sortPriority=4}
  n3 = VectorParameter() {group=Anisotropy, parameterName=Anisotropy UV Scale, sortPriority=5, defaultValue=<rgba 1 1 0 0>}
  n4 = MaterialFunctionCall!(
   TexCoord=ScalarParameter() {group=Anisotropy, parameterName=Anisotropy UV Index, sliderMax=7, sortPriority=3}
   Offset_X=@n2.R
   Offset_Y=@n2.R
   Scale_X=@n3.R
   Scale_Y=@n3.R
   Rotation=ScalarParameter() {group=Anisotropy, parameterName=Anisotropy UV Rotation, sliderMax=6.2832, sortPriority=6}
   TilingMethod=VectorParameter() {group=Anisotropy, parameterName=Anisotropy Tiling Method, sortPriority=5}.RGB
  ) {materialFunction=&/MF_GLTF_UVTransform}
  n5 = TextureSampleParameter2D!(UVs=@n4.UVs, "DDX(UVs)"=@n4.UVs, "DDY(UVs)"=@n4.UVs, Apply View MipBias) {
   group=Anisotropy
   mipValueMode=TMVM_Derivative
   parameterName=Anisotropy Texture
   samplerSource=SSM_Clamp_WorldGroupSettings
   samplerType=SAMPLERTYPE_LinearColor
   sortPriority=2
   texture=/GLTFExporter/Textures/Proxy/T_Anisotropy_Direction_Linear
  }
  n6 = ComponentMask(None=@n5.RGB) {g=true, r=true}
  MaterialExpressionAppendVector_0 = AppendVector(A=Subtract(A=Multiply(A=@n6) {constB=2}), B=Constant())
  n7 = Custom!(
   Rotation=@MaterialExpressionScalarParameter_19
   SurfaceTangent=@MaterialExpressionCustom_1
   SurfaceNormal=@MaterialExpressionFunctionInput_2
   Direction=@MaterialExpressionAppendVector_0
  ) {
   code="const float3 N = SurfaceNormal;\r\nconst float3 T = SurfaceTangent;\r\nconst float3 B = normalize(cross(SurfaceTangent, SurfaceNormal));\r\n\r\nconst float cosAngle = cos(Rotation);\r\nconst float sinAngle = sin(Rotation);\r\n\r\nDirection.xy = mul(float2x2(cosAngle, sinAngle, -sinAngle, cosAngle), normalize(Direction.xy));\r\n\r\nconst float3 Out = normalize(mul(float3x3(T, B, N), Direction));\r\n\r\nreturn Out;"
   description=GetAnisotropyTangent
   showCode=true
   inputs [
    {inputName=Rotation, input {expression=&self:MaterialExpressionScalarParameter_19}}
    {inputName=SurfaceTangent, input {expression=&self:MaterialExpressionCustom_1}}
    {inputName=SurfaceNormal, input {expression=&self:MaterialExpressionFunctionInput_2}}
    {inputName=Direction, input {expression=&self:MaterialExpressionAppendVector_0}}
   ]
  }
  FunctionOutput(None=@n7) {outputName=Tangent, sortPriority=1}
  FunctionOutput(None=Multiply(A=@n5.B, B=ScalarParameter() {group=Anisotropy, parameterName=Anisotropy Strength, sortPriority=0})) {outputName=Anisotropy}
 ]}
