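# M_GLTF_Default is the default-lit proxy material of the GLTFExporter plugin, enabled for static lighting, skeletal meshes and mesh deformers.
# It is built entirely from shared MF_GLTF proxy functions for metallic-roughness, base color, normal, emissive, occlusion, specular, specular anisotropy, IOR and iridescence.
# Base color and IOR feed the iridescence function, keeping each glTF material feature in its own reusable function.
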Material /GLTFExporter/Materials/Proxy/M_GLTF_Default
graph {
 properties {
  bUsedWithMeshDeformer=true
  bUsedWithSkeletalMesh=true
  bUsedWithStaticLighting=true
  opacityMaskClipValue=0.5
  pixelDepthOffsetMode=PDOM_Legacy
  refractionMethod=RM_IndexOfRefraction
  usedShadingModels=Default Lit
 }
 nodes [
  MaterialFunctionCall() {materialFunction=/GLTFExporter/Functions/Proxy/MF_GLTF_MetallicRoughness}
  MaterialFunctionCall() {materialFunction=/GLTFExporter/Functions/Proxy/MF_GLTF_Emissive}
  MaterialFunctionCall() {materialFunction=/GLTFExporter/Functions/Proxy/MF_GLTF_Normal}
  MaterialFunctionCall() {materialFunction=/GLTFExporter/Functions/Proxy/MF_GLTF_Occlusion}
  MaterialFunctionCall() {materialFunction=/GLTFExporter/Functions/Proxy/MF_GLTF_Specular}
  MaterialFunctionCall!(Normal) {materialFunction=/GLTFExporter/Functions/Proxy/MF_GLTF_SpecularAnisotropy}
  MaterialFunctionCall!(
   BaseColor=MaterialFunctionCall() {materialFunction=/GLTFExporter/Functions/Proxy/MF_GLTF_BaseColor}.BaseColor
   IOR=MaterialFunctionCall() {materialFunction=/GLTFExporter/Functions/Proxy/MF_GLTF_IOR}.Result
  ) {materialFunction=/GLTFExporter/Functions/Proxy/MF_GLTF_Iridescence}
 ]
}
