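# M_GLTF_Unlit is an unlit proxy material from the GLTFExporter plugin.
# Its only node is a call to the MF_GLTF_BaseColor material function, with no other shading logic in the graph.
# It is flagged for use with skeletal meshes and mesh deformers and has an opacity mask clip value of one half.
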Material /GLTFExporter/Materials/Proxy/M_GLTF_Unlit
graph {
 properties {
  bUsedWithMeshDeformer=true
  bUsedWithSkeletalMesh=true
  opacityMaskClipValue=0.5
  pixelDepthOffsetMode=PDOM_Legacy
  shadingModel=MSM_Unlit
  usedShadingModels=Unlit
 }
 nodes [MaterialFunctionCall() {materialFunction=/GLTFExporter/Functions/Proxy/MF_GLTF_BaseColor}]
}
